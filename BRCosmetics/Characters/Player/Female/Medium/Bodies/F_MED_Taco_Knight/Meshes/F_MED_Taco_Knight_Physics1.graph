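# Physics asset for the female medium Taco Knight cosmetic body, defining secondary motion for its cloth and dangling accessories.
# Twelve three-bone cloth chains hang from the thigh rot bones or pelvis with tight swing and twist limits.
# Waist bag, shoulder pads and "drug" prop constrain to their base bones, with potion and three chilis attached to the drug.
# Unconstrained spine and thigh capsules supply body collision shapes.
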PhysicsAsset /BRCosmetics/Characters/Player/Female/Medium/Bodies/F_MED_Taco_Knight/Meshes/F_MED_Taco_Knight_Physics1
bodies {
 drug [{shapeType=Capsule, length=9.6753, radius=4.202, center=<-3.0405 -0.724 -4.3952>, rotation=<rot 3.3363 -90 -2.1501>}]
 drug_base [{shapeType=Sphere, radius=2.6602}]
 drug_chili01 [{shapeType=Capsule, length=5.0594, radius=0.6383, center=<-0.2389 -2.7369 -0.0521>, rotation=<rot 0 -7.499 -89.4009>}]
 drug_chili02 [{shapeType=Capsule, length=4.7813, radius=0.5396, center=<-0.2005 -2.619 -0.1454>, rotation=<rot 0 176.1792 -86.3487>}]
 drug_chili03 [{shapeType=Capsule, length=4.1748, radius=0.5677, center=<-0.1164 -2.072 -0.0907>, rotation=<rot 0 179.0884 -84.8196>}]
 drug_potion [{shapeType=Capsule, length=4.6038, radius=0.7027, center=<0.091 0.063 -0.38>, rotation=<rot 0 -82.754 -86.9594>}]
 dyn_clothA_l_01 [{shapeType=Capsule, length=7.5894, radius=1.5568, center=<0 0 -1.7808>}]
 dyn_clothA_l_02 [{shapeType=Capsule, length=7.5894, radius=1.5568, center=<0 0 -1.7808>}]
 dyn_clothA_l_03 [{shapeType=Capsule, length=7.5894, radius=1.5568, center=<0 0 -1.7808>}]
 dyn_clothA_r_01 [{shapeType=Capsule, length=6.8625, radius=2.2837, center=<0.0001 0 1.998>}]
 dyn_clothA_r_02 [{shapeType=Capsule, length=6.8625, radius=2.2837, center=<0.0001 0 1.998>}]
 dyn_clothA_r_03 [{shapeType=Capsule, length=6.8625, radius=2.2837, center=<0.0001 0 1.998>}]
 dyn_clothB_l_01 [{shapeType=Capsule, length=4.8821, radius=1.6375, center=<0 0 -1.7808>}]
 dyn_clothB_l_02 [{shapeType=Capsule, length=5.4078, radius=1.1119, center=<0 0 -1.7808>}]
 dyn_clothB_l_03 [{shapeType=Capsule, length=5.4078, radius=1.1119, center=<0 0 -1.7808>}]
 dyn_clothB_r_01 [{shapeType=Capsule, length=2.5921, radius=3.9275, center=<0.0001 0 1.998>}]
 dyn_clothB_r_02 [{shapeType=Capsule, length=2.5921, radius=3.9275, center=<0.0001 0 1.998>}]
 dyn_clothB_r_03 [{shapeType=Capsule, length=2.5921, radius=3.9275, center=<0.0001 0 1.998>}]
 dyn_clothC_l_01 [{shapeType=Capsule, length=4.176, radius=2.3436, center=<0 0 -2.5371>}]
 dyn_clothC_l_02 [{shapeType=Capsule, length=4.176, radius=2.3436, center=<0 0 -2.5371>}]
 dyn_clothC_l_03 [{shapeType=Capsule, length=4.176, radius=2.3436, center=<0 0 -2.5371>}]
 dyn_clothC_r_01 [{shapeType=Capsule, length=4.176, radius=2.3436, center=<0.0001 0 0.753>}]
 dyn_clothC_r_02 [{shapeType=Capsule, length=4.176, radius=2.3436, center=<0.0001 0 0.753>}]
 dyn_clothC_r_03 [{shapeType=Capsule, length=4.176, radius=2.3436, center=<0.0001 0 0.753>}]
 dyn_clothD_l_01 [{shapeType=Capsule, length=4.176, radius=2.3436, center=<0 0 -2.5371>}]
 dyn_clothD_l_02 [{shapeType=Capsule, length=4.176, radius=2.3436, center=<0 0 -2.5371>}]
 dyn_clothD_l_03 [{shapeType=Capsule, length=4.176, radius=2.3436, center=<0 0 -2.5371>}]
 dyn_clothD_r_01 [{shapeType=Capsule, length=4.176, radius=2.3436, center=<0.0001 0 0.753>}]
 dyn_clothD_r_02 [{shapeType=Capsule, length=4.176, radius=2.3436, center=<0.0001 0 0.753>}]
 dyn_clothD_r_03 [{shapeType=Capsule, length=4.176, radius=2.3436, center=<0.0001 0 0.753>}]
 dyn_clothE_l_01 [{shapeType=Capsule, length=4.176, radius=2.3436, center=<0 0 -2.5371>}]
 dyn_clothE_l_02 [{shapeType=Capsule, length=4.176, radius=2.3436, center=<0 0 -2.5371>}]
 dyn_clothE_l_03 [{shapeType=Capsule, length=4.176, radius=2.3436, center=<0 0 -2.5371>}]
 dyn_clothE_r_01 [{shapeType=Capsule, length=4.176, radius=2.3436, center=<0.0001 0 0.753>}]
 dyn_clothE_r_02 [{shapeType=Capsule, length=4.176, radius=2.3436, center=<0.0001 0 0.753>}]
 dyn_clothE_r_03 [{shapeType=Capsule, length=4.176, radius=2.3436, center=<0.0001 0 0.753>}]
 dyn_cloth_l_01 [{shapeType=Capsule, length=6.8625, radius=2.2837, center=<0 0 -1.7808>}]
 dyn_cloth_l_02 [{shapeType=Capsule, length=6.8625, radius=2.2837, center=<0 0 -1.7808>}]
 dyn_cloth_l_03 [{shapeType=Capsule, length=6.8625, radius=2.2837, center=<0 0 -1.7808>}]
 dyn_cloth_r_01 [{shapeType=Capsule, length=6.8625, radius=2.2837, center=<0.0001 0 1.998>}]
 dyn_cloth_r_02 [{shapeType=Capsule, length=6.8625, radius=2.2837, center=<0.0001 0 1.998>}]
 dyn_cloth_r_03 [{shapeType=Capsule, length=6.8625, radius=2.2837, center=<0.0001 0 1.998>}]
 pelvis [{shapeType=Capsule, length=14.4016, radius=8.3838, center=<-9.1833 -2.0249 0.2907>, rotation=<rot 6.2427 -86.041 1.2346>}]
 shldr_pad_l [{shapeType=Capsule, length=11.9099, radius=8.0959, center=<6.8791 0.4647 -2.176>, rotation=<rot -0.4923 88.4491 -89.9915>}]
 shldr_pad_l_base [{shapeType=Capsule, length=2.4966, radius=1.8463}]
 shldr_pad_r [{shapeType=Capsule, length=11.9099, radius=8.0959, center=<-6.8791 -0.4647 2.1758>, rotation=<rot -0.4923 88.4491 -89.9915>}]
 shldr_pad_r_base [{shapeType=Capsule, length=2.4966, radius=1.8463}]
 spine_01 [{shapeType=Capsule, length=9.7481, radius=8.328, center=<2.7583 0.0002 0>, rotation=<rot 0 -90 0>}]
 spine_05 [{shapeType=Capsule, length=26.9346, radius=15, center=<-0.0003 -24.3681 0>}]
 thigh_l [{shapeType=Capsule, length=24.0945, radius=7.7782, center=<-19.79 -1.2391 0.2762>, rotation=<rot -0.0239 -89.481 -89.5705>}]
 thigh_l_rot [{shapeType=Sphere, radius=4.0997}]
 thigh_r [{shapeType=Capsule, length=23.9734, radius=7.8992, center=<22.6763 1.052 -0.0795>, rotation=<rot -0.0239 -89.4809 -91.7238>}]
 thigh_r_rot [{shapeType=Sphere, radius=4.0997}]
 waist_bag [{shapeType=Capsule, length=6.0602, radius=4.0866, center=<0.2162 3.3385 -2.7109>, rotation=<rot -13.6966 -90 -4.6814>}]
 waist_bag_base [{shapeType=Sphere, radius=1.9242}]
}
constraints [
 {bone1Name=dyn_cloth_l_01, bone2Name=thigh_l_rot, swing1LimitDegrees=10, swing2LimitDegrees=10, twistLimitDegrees=20}
 {bone1Name=dyn_cloth_l_02, bone2Name=dyn_cloth_l_01, swing1LimitDegrees=10, swing2LimitDegrees=10, twistLimitDegrees=20}
 {bone1Name=dyn_cloth_l_03, bone2Name=dyn_cloth_l_02, swing1LimitDegrees=10, swing2LimitDegrees=10, twistLimitDegrees=20}
 {bone1Name=dyn_clothA_l_01, bone2Name=thigh_l_rot, swing1LimitDegrees=10, swing2LimitDegrees=10, twistLimitDegrees=20}
 {bone1Name=dyn_clothA_l_02, bone2Name=dyn_clothA_l_01, swing1LimitDegrees=10, swing2LimitDegrees=10, twistLimitDegrees=20}
 {bone1Name=dyn_clothA_l_03, bone2Name=dyn_clothA_l_02, swing1LimitDegrees=10, swing2LimitDegrees=10, twistLimitDegrees=20}
 {bone1Name=dyn_clothB_l_01, bone2Name=thigh_l_rot, swing1LimitDegrees=10, swing2LimitDegrees=10, twistLimitDegrees=10}
 {bone1Name=dyn_clothB_l_02, bone2Name=dyn_clothB_l_01, swing1LimitDegrees=10, swing2LimitDegrees=10, twistLimitDegrees=10}
 {bone1Name=dyn_clothB_l_03, bone2Name=dyn_clothB_l_02, swing1LimitDegrees=10, swing2LimitDegrees=10, twistLimitDegrees=10}
 {bone1Name=dyn_cloth_r_01, bone2Name=thigh_r_rot, swing1LimitDegrees=10, swing2LimitDegrees=10, twistLimitDegrees=10}
 {bone1Name=dyn_clothC_l_02, bone2Name=dyn_clothC_l_01, swing1LimitDegrees=10, swing2LimitDegrees=10, twistLimitDegrees=10}
 {bone1Name=dyn_clothC_l_03, bone2Name=dyn_clothC_l_02, swing1LimitDegrees=10, swing2LimitDegrees=10, twistLimitDegrees=10}
 {bone1Name=dyn_cloth_r_02, bone2Name=dyn_cloth_r_01, swing1LimitDegrees=10, swing2LimitDegrees=10, twistLimitDegrees=10}
 {bone1Name=dyn_clothD_l_02, bone2Name=dyn_clothD_l_01, swing1LimitDegrees=10, swing2LimitDegrees=10, twistLimitDegrees=10}
 {bone1Name=dyn_clothD_l_03, bone2Name=dyn_clothD_l_02, swing1LimitDegrees=10, swing2LimitDegrees=10, twistLimitDegrees=10}
 {bone1Name=dyn_cloth_r_03, bone2Name=dyn_cloth_r_02, swing1LimitDegrees=10, swing2LimitDegrees=10, twistLimitDegrees=10}
 {bone1Name=dyn_clothE_l_02, bone2Name=dyn_clothE_l_01, swing1LimitDegrees=10, swing2LimitDegrees=10, twistLimitDegrees=10}
 {bone1Name=dyn_clothE_l_03, bone2Name=dyn_clothE_l_02, swing1LimitDegrees=10, swing2LimitDegrees=10, twistLimitDegrees=10}
 {bone1Name=dyn_clothA_r_01, bone2Name=thigh_r_rot, swing1LimitDegrees=10, swing2LimitDegrees=10, twistLimitDegrees=10}
 {bone1Name=dyn_clothA_r_02, bone2Name=dyn_clothA_r_01, swing1LimitDegrees=10, swing2LimitDegrees=10, twistLimitDegrees=10}
 {bone1Name=dyn_clothA_r_03, bone2Name=dyn_clothA_r_02, swing1LimitDegrees=10, swing2LimitDegrees=10, twistLimitDegrees=10}
 {bone1Name=dyn_clothB_r_01, bone2Name=thigh_r_rot, swing1LimitDegrees=10, swing2LimitDegrees=10, twistLimitDegrees=10}
 {bone1Name=dyn_clothB_r_02, bone2Name=dyn_clothB_r_01, swing1LimitDegrees=10, swing2LimitDegrees=10, twistLimitDegrees=10}
 {bone1Name=dyn_clothB_r_03, bone2Name=dyn_clothB_r_02, swing1LimitDegrees=10, swing2LimitDegrees=10, twistLimitDegrees=10}
 {bone1Name=dyn_clothC_r_02, bone2Name=dyn_clothC_r_01, swing1LimitDegrees=10, swing2LimitDegrees=10, twistLimitDegrees=10}
 {bone1Name=dyn_clothC_r_03, bone2Name=dyn_clothC_r_02, swing1LimitDegrees=10, swing2LimitDegrees=10, twistLimitDegrees=10}
 {bone1Name=dyn_clothD_r_02, bone2Name=dyn_clothD_r_01, swing1LimitDegrees=10, swing2LimitDegrees=10, twistLimitDegrees=10}
 {bone1Name=dyn_clothD_r_03, bone2Name=dyn_clothD_r_02, swing1LimitDegrees=10, swing2LimitDegrees=10, twistLimitDegrees=10}
 {bone1Name=dyn_clothE_r_02, bone2Name=dyn_clothE_r_01, swing1LimitDegrees=10, swing2LimitDegrees=10, twistLimitDegrees=10}
 {bone1Name=dyn_clothE_r_03, bone2Name=dyn_clothE_r_02, swing1LimitDegrees=10, swing2LimitDegrees=10, twistLimitDegrees=10}
 {bone1Name=dyn_clothC_l_01, bone2Name=pelvis, swing1LimitDegrees=10, swing2LimitDegrees=10, twistLimitDegrees=10}
 {bone1Name=dyn_clothD_l_01, bone2Name=pelvis, swing1LimitDegrees=10, swing2LimitDegrees=10, twistLimitDegrees=10}
 {bone1Name=dyn_clothE_l_01, bone2Name=pelvis, swing1LimitDegrees=10, swing2LimitDegrees=10, twistLimitDegrees=10}
 {bone1Name=dyn_clothC_r_01, bone2Name=pelvis, swing1LimitDegrees=10, swing2LimitDegrees=10, twistLimitDegrees=10}
 {bone1Name=dyn_clothD_r_01, bone2Name=pelvis, swing1LimitDegrees=10, swing2LimitDegrees=10, twistLimitDegrees=10}
 {bone1Name=dyn_clothE_r_01, bone2Name=pelvis, swing1LimitDegrees=10, swing2LimitDegrees=10, twistLimitDegrees=10}
 {bone1Name=waist_bag, bone2Name=waist_bag_base, swing1LimitDegrees=5, swing2LimitDegrees=5, twistLimitDegrees=2}
 {bone1Name=drug, bone2Name=drug_base, swing1LimitDegrees=20, swing2LimitDegrees=40, twistLimitDegrees=1}
 {bone1Name=drug_potion, bone2Name=drug, swing1LimitDegrees=20, swing2LimitDegrees=20, twistLimitDegrees=20}
 {bone1Name=drug_chili01, bone2Name=drug, swing1LimitDegrees=20, swing2LimitDegrees=20, twistLimitDegrees=20}
 {bone1Name=drug_chili02, bone2Name=drug, swing1LimitDegrees=10, swing2LimitDegrees=10, twistLimitDegrees=10}
 {bone1Name=drug_chili03, bone2Name=drug, swing1LimitDegrees=15, swing2LimitDegrees=15, twistLimitDegrees=15}
 {bone1Name=shldr_pad_l, bone2Name=shldr_pad_l_base, swing1LimitDegrees=1, swing2LimitDegrees=1, twistLimitDegrees=1}
 {bone1Name=shldr_pad_r, bone2Name=shldr_pad_r_base, swing1LimitDegrees=1, swing2LimitDegrees=1, twistLimitDegrees=1}
]
