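# Physics asset for the female medium Taco Knight body, named for skydiving, that covers its dynamic cloth bone chains.
# Capsule chains on each side sit alongside pelvis, lower spine and thigh bodies.
# The base, A, B and C chains have narrow limits and hang from thigh rotation bones, except right C, which hangs from the pelvis.
# D and E chains start at their own unparented root bodies and allow almost no swing but wider twist.
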PhysicsAsset /BRCosmetics/Characters/Player/Female/Medium/Bodies/F_MED_Taco_Knight/Meshes/F_MED_Taco_Knight_Physics_SkyDiving
bodies {
 dyn_clothA_l_01 [{shapeType=Capsule, length=5.8026, radius=3.3436, center=<0 0 -1.7808>}]
 dyn_clothA_l_02 [{shapeType=Capsule, length=5.1588, radius=3.9874, center=<0 0 -1.7808>}]
 dyn_clothA_l_03 [{shapeType=Capsule, length=5.1525, radius=3.9937, center=<0 0 -1.7808>}]
 dyn_clothA_r_01 [{shapeType=Capsule, length=6.8625, radius=2.2837, center=<0.0001 0 1.998>}]
 dyn_clothA_r_02 [{shapeType=Capsule, length=6.8625, radius=2.2837, center=<0.0001 0 1.998>}]
 dyn_clothA_r_03 [{shapeType=Capsule, length=6.8625, radius=2.2837, center=<0.0001 0 1.998>}]
 dyn_clothB_l_01 [{shapeType=Capsule, length=3.5481, radius=2.9715, center=<0 0 -1.7808>}]
 dyn_clothB_l_02 [{shapeType=Capsule, length=3.5716, radius=2.9481, center=<0 0 -1.7808>}]
 dyn_clothB_l_03 [{shapeType=Capsule, length=3.5716, radius=2.9481, center=<0 0 -1.7808>}]
 dyn_clothB_r_01 [{shapeType=Capsule, length=4.2843, radius=2.2353, center=<0.0001 0 1.998>}]
 dyn_clothB_r_02 [{shapeType=Capsule, length=4.3441, radius=2.1756, center=<0.0001 0 1.998>}]
 dyn_clothB_r_03 [{shapeType=Capsule, length=4.3441, radius=2.1756, center=<0.0001 0 1.998>}]
 dyn_clothC_l_01 [{shapeType=Capsule, length=4.176, radius=2.3436, center=<0 0 -2.5371>}]
 dyn_clothC_l_02 [{shapeType=Capsule, length=4.176, radius=2.3436, center=<0 0 -2.5371>}]
 dyn_clothC_l_03 [{shapeType=Capsule, length=4.176, radius=2.3436, center=<0 0 -2.5371>}]
 dyn_clothC_r_01 [{shapeType=Capsule, length=4.176, radius=2.3436, center=<0.0001 0 0.753>}]
 dyn_clothC_r_02 [{shapeType=Capsule, length=4.176, radius=2.3436, center=<0.0001 0 0.753>}]
 dyn_clothC_r_03 [{shapeType=Capsule, length=4.176, radius=2.3436, center=<0.0001 0 0.753>}]
 dyn_clothD_l_00 [{shapeType=Capsule, length=0, radius=2.2369}]
 dyn_clothD_l_01 [{shapeType=Capsule, length=4.176, radius=2.3436, center=<0 0 -2.5371>}]
 dyn_clothD_l_02 [{shapeType=Capsule, length=4.176, radius=2.3436, center=<0 0 -2.5371>}]
 dyn_clothD_l_03 [{shapeType=Capsule, length=4.176, radius=2.3436, center=<0 0 -2.5371>}]
 dyn_clothD_r_00 [{shapeType=Capsule, length=0, radius=2.2369}]
 dyn_clothD_r_01 [{shapeType=Capsule, length=4.176, radius=2.3436, center=<0.0001 0 0.753>}]
 dyn_clothD_r_02 [{shapeType=Capsule, length=4.176, radius=2.3436, center=<0.0001 0 0.753>}]
 dyn_clothD_r_03 [{shapeType=Capsule, length=4.176, radius=2.3436, center=<0.0001 0 0.753>}]
 dyn_clothE_l_00 [{shapeType=Capsule, length=0, radius=2.2369}]
 dyn_clothE_l_01 [{shapeType=Capsule, length=4.176, radius=2.3436, center=<0 0 -2.5371>}]
 dyn_clothE_l_02 [{shapeType=Capsule, length=4.176, radius=2.3436, center=<0 0 -2.5371>}]
 dyn_clothE_l_03 [{shapeType=Capsule, length=4.176, radius=2.3436, center=<0 0 -2.5371>}]
 dyn_clothE_r_00 [{shapeType=Capsule, length=0, radius=2.2369}]
 dyn_clothE_r_01 [{shapeType=Capsule, length=4.176, radius=2.3436, center=<0.0001 0 0.753>}]
 dyn_clothE_r_02 [{shapeType=Capsule, length=4.176, radius=2.3436, center=<0.0001 0 0.753>}]
 dyn_clothE_r_03 [{shapeType=Capsule, length=4.176, radius=2.3436, center=<0.0001 0 0.753>}]
 dyn_cloth_l_01 [{shapeType=Capsule, length=6.8625, radius=2.2837, center=<0 0 -1.7808>}]
 dyn_cloth_l_02 [{shapeType=Capsule, length=6.8625, radius=2.2837, center=<0 0 -1.7808>}]
 dyn_cloth_l_03 [{shapeType=Capsule, length=6.8625, radius=2.2837, center=<0 0 -1.7808>}]
 dyn_cloth_r_01 [{shapeType=Capsule, length=6.8625, radius=2.2837, center=<0.0001 0 1.998>}]
 dyn_cloth_r_02 [{shapeType=Capsule, length=6.8625, radius=2.2837, center=<0.0001 0 1.998>}]
 dyn_cloth_r_03 [{shapeType=Capsule, length=6.8625, radius=2.2837, center=<0.0001 0 1.998>}]
 pelvis [{shapeType=Capsule, length=15.1297, radius=9.184, center=<-12.0605 -2.2173 0.3524>, rotation=<rot 6.2427 -86.041 1.2346>}]
 spine_01 [{shapeType=Capsule, length=9.7481, radius=8.328, center=<2.7583 0.0002 0>, rotation=<rot 0 -90 0>}]
 thigh_l_rot [{shapeType=Sphere, radius=4.0997}]
 thigh_r_rot [{shapeType=Sphere, radius=4.0997}]
 thigh_twist_01_l [{shapeType=Capsule, length=22.3186, radius=11.364, center=<6.5391 -1.1144 2.2108>, rotation=<rot 89.7338 0.006 0.5638>}]
 thigh_twist_01_r [{shapeType=Capsule, length=22.9652, radius=10.7173, center=<-7.791 -0.4118 -1.163>, rotation=<rot 89.7338 0.0061 0.5641>}]
}
constraints [
 {bone1Name=dyn_cloth_l_01, bone2Name=thigh_l_rot, swing1LimitDegrees=10, swing2LimitDegrees=10, twistLimitDegrees=20}
 {bone1Name=dyn_cloth_l_02, bone2Name=dyn_cloth_l_01, swing1LimitDegrees=10, swing2LimitDegrees=10, twistLimitDegrees=20}
 {bone1Name=dyn_cloth_l_03, bone2Name=dyn_cloth_l_02, swing1LimitDegrees=10, swing2LimitDegrees=10, twistLimitDegrees=20}
 {bone1Name=dyn_clothA_l_01, bone2Name=thigh_l_rot, swing1LimitDegrees=10, swing2LimitDegrees=10, twistLimitDegrees=20}
 {bone1Name=dyn_clothA_l_02, bone2Name=dyn_clothA_l_01, swing1LimitDegrees=10, swing2LimitDegrees=10, twistLimitDegrees=20}
 {bone1Name=dyn_clothA_l_03, bone2Name=dyn_clothA_l_02, swing1LimitDegrees=10, swing2LimitDegrees=10, twistLimitDegrees=20}
 {bone1Name=dyn_clothB_l_01, bone2Name=thigh_l_rot, swing1LimitDegrees=10, swing2LimitDegrees=10, twistLimitDegrees=10}
 {bone1Name=dyn_clothB_l_02, bone2Name=dyn_clothB_l_01, swing1LimitDegrees=10, swing2LimitDegrees=10, twistLimitDegrees=10}
 {bone1Name=dyn_clothB_l_03, bone2Name=dyn_clothB_l_02, swing1LimitDegrees=10, swing2LimitDegrees=10, twistLimitDegrees=10}
 {bone1Name=dyn_cloth_r_01, bone2Name=thigh_r_rot, swing1LimitDegrees=10, swing2LimitDegrees=10, twistLimitDegrees=10}
 {bone1Name=dyn_clothC_l_02, bone2Name=dyn_clothC_l_01, swing1LimitDegrees=10, swing2LimitDegrees=10, twistLimitDegrees=10}
 {bone1Name=dyn_clothC_l_03, bone2Name=dyn_clothC_l_02, swing1LimitDegrees=10, swing2LimitDegrees=10, twistLimitDegrees=10}
 {bone1Name=dyn_cloth_r_02, bone2Name=dyn_cloth_r_01, swing1LimitDegrees=10, swing2LimitDegrees=10, twistLimitDegrees=10}
 {bone1Name=dyn_clothD_l_02, bone2Name=dyn_clothD_l_01, swing1LimitDegrees=1, swing2LimitDegrees=1, twistLimitDegrees=35}
 {bone1Name=dyn_clothD_l_03, bone2Name=dyn_clothD_l_02, swing1LimitDegrees=1, swing2LimitDegrees=1, twistLimitDegrees=35}
 {bone1Name=dyn_cloth_r_03, bone2Name=dyn_cloth_r_02, swing1LimitDegrees=10, swing2LimitDegrees=10, twistLimitDegrees=10}
 {bone1Name=dyn_clothE_l_02, bone2Name=dyn_clothE_l_01, swing1LimitDegrees=1, swing2LimitDegrees=1, twistLimitDegrees=35}
 {bone1Name=dyn_clothE_l_03, bone2Name=dyn_clothE_l_02, swing1LimitDegrees=1, swing2LimitDegrees=1, twistLimitDegrees=35}
 {bone1Name=dyn_clothA_r_01, bone2Name=thigh_r_rot, swing1LimitDegrees=10, swing2LimitDegrees=10, twistLimitDegrees=10}
 {bone1Name=dyn_clothA_r_02, bone2Name=dyn_clothA_r_01, swing1LimitDegrees=10, swing2LimitDegrees=10, twistLimitDegrees=10}
 {bone1Name=dyn_clothA_r_03, bone2Name=dyn_clothA_r_02, swing1LimitDegrees=10, swing2LimitDegrees=10, twistLimitDegrees=10}
 {bone1Name=dyn_clothB_r_01, bone2Name=thigh_r_rot, swing1LimitDegrees=10, swing2LimitDegrees=10, twistLimitDegrees=10}
 {bone1Name=dyn_clothB_r_02, bone2Name=dyn_clothB_r_01, swing1LimitDegrees=10, swing2LimitDegrees=10, twistLimitDegrees=10}
 {bone1Name=dyn_clothB_r_03, bone2Name=dyn_clothB_r_02, swing1LimitDegrees=10, swing2LimitDegrees=10, twistLimitDegrees=10}
 {bone1Name=dyn_clothC_r_02, bone2Name=dyn_clothC_r_01, swing1LimitDegrees=10, swing2LimitDegrees=10, twistLimitDegrees=10}
 {bone1Name=dyn_clothC_r_03, bone2Name=dyn_clothC_r_02, swing1LimitDegrees=10, swing2LimitDegrees=10, twistLimitDegrees=10}
 {bone1Name=dyn_clothD_r_02, bone2Name=dyn_clothD_r_01, swing1LimitDegrees=1, swing2LimitDegrees=1, twistLimitDegrees=35}
 {bone1Name=dyn_clothD_r_03, bone2Name=dyn_clothD_r_02, swing1LimitDegrees=1, swing2LimitDegrees=1, twistLimitDegrees=35}
 {bone1Name=dyn_clothE_r_02, bone2Name=dyn_clothE_r_01, swing1LimitDegrees=1, swing2LimitDegrees=1, twistLimitDegrees=35}
 {bone1Name=dyn_clothE_r_03, bone2Name=dyn_clothE_r_02, swing1LimitDegrees=1, swing2LimitDegrees=1, twistLimitDegrees=35}
 {bone1Name=dyn_clothC_l_01, bone2Name=thigh_l_rot, swing1LimitDegrees=10, swing2LimitDegrees=10, twistLimitDegrees=10}
 {bone1Name=dyn_clothD_l_01, bone2Name=dyn_clothD_l_00, swing1LimitDegrees=1, swing2LimitDegrees=1, twistLimitDegrees=35}
 {bone1Name=dyn_clothE_l_01, bone2Name=dyn_clothE_l_00, swing1LimitDegrees=1, swing2LimitDegrees=1, twistLimitDegrees=35}
 {bone1Name=dyn_clothC_r_01, bone2Name=pelvis, swing1LimitDegrees=10, swing2LimitDegrees=10, twistLimitDegrees=10}
 {bone1Name=dyn_clothD_r_01, bone2Name=dyn_clothD_r_00, swing1LimitDegrees=1, swing2LimitDegrees=1, twistLimitDegrees=35}
 {bone1Name=dyn_clothE_r_01, bone2Name=dyn_clothE_r_00, swing1LimitDegrees=1, swing2LimitDegrees=1, twistLimitDegrees=35}
]
