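# Post-process style cosmetic AnimBP for the female Taco Knight body, adding corrective and secondary motion to armor, cloth and waist bag.
# It copies the attached parent's pose, then caches it after bone constraints and additive corrections weighted by limb rotation deltas computed each update.
# RigidBody branches blend between normal, with a crouch bag offset, and skydiving, then feed a final RigidBody pass.
# A separate clothD/E constraint chain is unconnected to the output.
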AnimBlueprint /BRCosmetics/Characters/Player/Female/Medium/Bodies/F_MED_Taco_Knight/Meshes/F_MED_Taco_Knight_AnimBP
target_skeleton=&/F_MED_Taco_Knight_Skeleton
graphs [
 AnimGraph {nodes [
   "Constraint - dyn_clothD_l_00 (PP)"!(ComponentPose, Alpha=1, ConstraintWeights_0=0.2, ConstraintWeights_1:float=0.8) -> (Pose) {
    alphaBoolBlend {blendOption=Linear}
    boneToModify {boneName=dyn_clothD_l_00}
    constraintSetup [{targetBone {boneName=thigh_twist_01_l}}, {targetBone {boneName=pelvis}}]
    constraintWeights [1,1]
   }
   n1 = "Constraint - dyn_clothE_l_00 (PP)"!(ComponentPose, Alpha=1, ConstraintWeights_0=0.5, ConstraintWeights_1:float=0.5) -> (Pose) {
    alphaBoolBlend {blendOption=Linear}
    boneToModify {boneName=dyn_clothE_l_00}
    constraintSetup [{targetBone {boneName=dyn_clothD_l_00}}, {targetBone {boneName=pelvis}}]
    constraintWeights [1,1]
   }
   n2 = "Constraint - dyn_clothE_r_00 (PP)"!(ComponentPose=@n1, Alpha=1, ConstraintWeights_0=0.5, ConstraintWeights_1:float=0.5) -> (Pose) {
    alphaBoolBlend {blendOption=Linear}
    boneToModify {boneName=dyn_clothE_r_00}
    constraintSetup [{targetBone {boneName=dyn_clothD_r_00}}, {targetBone {boneName=pelvis}}]
    constraintWeights [1,1]
   }
   "Constraint - dyn_clothD_r_00 (PP)"!(ComponentPose=@n2, Alpha=1, ConstraintWeights_0=0.2, ConstraintWeights_1:float=0.8) -> (Pose) {
    alphaBoolBlend {blendOption=Linear}
    boneToModify {boneName=dyn_clothD_r_00}
    constraintSetup [{targetBone {boneName=dyn_clothC_r_01}}, {targetBone {boneName=dyn_clothE_r_00}}]
    constraintWeights [1,1]
   }
   n3 = "Constraint - shldr_pad_l_base (PP)"!(ComponentPose=Local To Component(LocalPose=Use cached pose 'AnimCache'()), Alpha=1, ConstraintWeights_0=0.3, ConstraintWeights_1:float=0.7) -> (Pose) {
    alphaBoolBlend {blendOption=Linear}
    boneToModify {boneName=shldr_pad_l_base}
    constraintSetup [{targetBone {boneName=clavicle_l}}, {targetBone {boneName=upperarm_l}}]
    constraintWeights [1,1]
   }
   n4 = "Constraint - shldr_pad_r_base (PP)"!(ComponentPose=@n3, Alpha=1, ConstraintWeights_0=0.3, ConstraintWeights_1:float=0.7) -> (Pose) {
    alphaBoolBlend {blendOption=Linear}
    boneToModify {boneName=shldr_pad_r_base}
    constraintSetup [{targetBone {boneName=clavicle_r}}, {targetBone {boneName=upperarm_r}}]
    constraintWeights [1,1]
   }
   n5 = RigidBody(ComponentPose=@n4, Alpha=0.4) {
    bClampLinearTranslationLimitToRefPose=true
    bOverrideWorldGravity=true
    overridePhysicsAsset=&/F_MED_Taco_Knight_Physics_SkyDiving
    simulationSpace=BaseBoneSpace
    alphaBoolBlend {blendOption=Linear}
    baseBoneRef {boneName=pelvis}
    componentAppliedLinearAccClamp=<500 500 500>
    componentLinearAccScale=<0.3 0.3 0.3>
    componentLinearVelScale=<0.3 0.3 0.3>
    overrideWorldGravity {z=-600}
   }
   n6 = Component To Local(ComponentPose=@n5)
   n7 = "Transform (Modify) Bone - Bone: waist_bag_base"(
    ComponentPose=Local To Component(LocalPose=Use cached pose 'AnimCache'())
    Alpha=Get leg_rot_l() -> (leg_rot_l:double)
    Translation="5,1,0"
    Rotation="0,-80,0"
    Scale="1,1,1"
   ) {
    rotationMode=BMM_Additive
    rotationSpace=BCS_BoneSpace
    scaleMode=BMM_Additive
    scaleSpace=BCS_BoneSpace
    translationMode=BMM_Additive
    translationSpace=BCS_BoneSpace
    alphaBoolBlend {blendOption=Linear}
    boneToModify {boneName=waist_bag_base}
    rotation {yaw=-80}
    translation=<5 1>
   }
   n8 = Component To Local(ComponentPose=@n7)
   n9 = Use cached pose 'AnimCache'()
   n10 = Get bIsCrouching()
   n11 = RigidBody(
    ComponentPose=Local To Component(LocalPose=Two Way Blend!(A=@n9, B=@n8, bAlphaBoolEnabled:bool=@n10) -> (Pose) {blendNode {alphaInputType=Bool, alphaBoolBlend {blendInTime=0.05, blendOption=Cubic, blendOutTime=0.1}}})
    Alpha=0.6
   ) {
    bClampLinearTranslationLimitToRefPose=true
    bOverrideWorldGravity=true
    overridePhysicsAsset=&/F_MED_Taco_Knight_Physics
    simulationSpace=BaseBoneSpace
    alphaBoolBlend {blendOption=Linear}
    baseBoneRef {boneName=pelvis}
    componentAppliedLinearAccClamp=<2000 2000 2000>
    componentLinearAccScale=<0.35 0.35 0.35>
    componentLinearVelScale=<0.3 0.3 0.3>
    overrideWorldGravity {z=-980}
   }
   n12 = RigidBody(ComponentPose=@n11, Alpha=0.3) {
    bClampLinearTranslationLimitToRefPose=true
    bOverrideWorldGravity=true
    overridePhysicsAsset=&/F_MED_Taco_Knight_Physics_F_cloth
    simulationSpace=BaseBoneSpace
    alphaBoolBlend {blendOption=Linear}
    baseBoneRef {boneName=pelvis}
    componentAppliedLinearAccClamp=<2000 2000 2000>
    componentLinearAccScale=<0.35 0.45 0.45>
    componentLinearVelScale=<0.3 0.3 0.3>
    overrideWorldGravity {z=-980}
   }
   n13 = Component To Local(ComponentPose=@n12)
   n14 = Get bIsSkyDiving()
   n15 = RigidBody(ComponentPose=Local To Component(LocalPose=Two Way Blend(A=@n13, B=@n6, bAlphaBoolEnabled=@n14) {blendNode {alphaInputType=Bool, alphaBoolBlend {blendInTime=0.05, blendOption=Cubic, blendOutTime=0.1}}})) {
    bClampLinearTranslationLimitToRefPose=true
    bOverrideWorldGravity=true
    overridePhysicsAsset=&/F_MED_Taco_Knight_Physics1
    simulationSpace=BaseBoneSpace
    alphaBoolBlend {blendOption=Linear}
    baseBoneRef {boneName=pelvis}
    componentAppliedLinearAccClamp=<1000 1000 1000>
    componentLinearAccScale=<0.35 0.45 0.45>
    componentLinearVelScale=<0.3 0.3 0.3>
    overrideWorldGravity {z=-980}
   }
   Output Pose(Result=Component To Local(ComponentPose=@n15)) {layerGroup=None}
   n16 = "Constraint - waist_bag_base (PP)"!(
    ComponentPose=Local To Component(LocalPose=Copy Pose From Mesh() {bUseAttachedParent=true})
    Alpha=1
    ConstraintWeights_0=0
    ConstraintWeights_1:float=1
   ) -> (Pose) {
    alphaBoolBlend {blendOption=Linear}
    boneToModify {boneName=waist_bag_base}
    constraintSetup [{targetBone {boneName=spine_02}}, {targetBone {boneName=pelvis}}]
    constraintWeights [1,1]
   }
   n17 = "Constraint - dyn_cloth_r_01 (PP)"!(ComponentPose=@n16, Alpha=1, ConstraintWeights_0=0.5, ConstraintWeights_1:float=0.5) -> (Pose) {
    alphaBoolBlend {blendOption=Linear}
    boneToModify {boneName=dyn_cloth_r_01}
    constraintSetup [{targetBone {boneName=dyn_cloth_r_01}}, {targetBone {boneName=dyn_cloth_l_01}}]
    constraintWeights [1,1]
   }
   n18 = "Constraint - upperarmDn_l (PP)"!(ComponentPose=@n17, Alpha=1, ConstraintWeights_0=0.2, ConstraintWeights_1:float=0.8) -> (Pose) {
    alphaBoolBlend {blendOption=Linear}
    boneToModify {boneName=upperarmDn_l}
    constraintSetup [{targetBone {boneName=clavicle_l}}, {targetBone {boneName=spine_05}}]
    constraintWeights [1,1]
   }
   n19 = "Constraint - upperarmDn_r (PP)"!(ComponentPose=@n18, Alpha=1, ConstraintWeights_0=0.2, ConstraintWeights_1:float=0.8) -> (Pose) {
    alphaBoolBlend {blendOption=Linear}
    boneToModify {boneName=upperarmDn_r}
    constraintSetup [{targetBone {boneName=clavicle_r}}, {targetBone {boneName=spine_05}}]
    constraintWeights [1,1]
   }
   n20 = "Constraint - thigh_l_rot (PP)"!(ComponentPose=@n19, Alpha=1, ConstraintWeights_0=1, ConstraintWeights_1:float=0) -> (Pose) {
    alphaBoolBlend {blendOption=Linear}
    boneToModify {boneName=thigh_l_rot}
    constraintSetup [{targetBone {boneName=thigh_twist_01_l}}, {targetBone {boneName=pelvis}}]
    constraintWeights [1,1]
   }
   n21 = "Constraint - thigh_r_rot (PP)"!(ComponentPose=@n20, Alpha=1, ConstraintWeights_0=0.9, ConstraintWeights_1:float=0.1) -> (Pose) {
    alphaBoolBlend {blendOption=Linear}
    boneToModify {boneName=thigh_r_rot}
    constraintSetup [{targetBone {boneName=thigh_twist_01_r}}, {targetBone {boneName=pelvis}}]
    constraintWeights [1,1]
   }
   n22 = "Constraint - dyn_cloth_l_01 (PP)"!(ComponentPose=@n21, Alpha=1, ConstraintWeights_0=0.5, ConstraintWeights_1:float=0.5) -> (Pose) {
    alphaBoolBlend {blendOption=Linear}
    boneToModify {boneName=dyn_cloth_l_01}
    constraintSetup [{targetBone {boneName=dyn_cloth_r_01}}, {targetBone {boneName=dyn_cloth_l_01}}]
    constraintWeights [1,1]
   }
   n23 = "Constraint - dyn_cloth_r_01 (PP)"(ComponentPose=@n22, Alpha=1, ConstraintWeights_0=0.5, ConstraintWeights_1=0.5) {
    alphaBoolBlend {blendOption=Linear}
    boneToModify {boneName=dyn_cloth_r_01}
    constraintSetup [{targetBone {boneName=dyn_cloth_l_01}}, {targetBone {boneName=dyn_cloth_r_01}}]
    constraintWeights [1,1]
   }
   n24 = "Constraint - dyn_clothC_l_01 (PP)"!(ComponentPose=@n23, Alpha=1, ConstraintWeights_0=0.3, ConstraintWeights_1:float=0.7) -> (Pose) {
    alphaBoolBlend {blendOption=Linear}
    boneToModify {boneName=dyn_clothC_l_01}
    constraintSetup [{targetBone {boneName=thigh_twist_01_l}}, {targetBone {boneName=dyn_clothD_r_00}}]
    constraintWeights [1,1]
   }
   n25 = "Constraint - dyn_clothC_r_01 (PP)"!(ComponentPose=@n24, Alpha=1, ConstraintWeights_0=0.5, ConstraintWeights_1:float=0.5) -> (Pose) {
    alphaBoolBlend {blendOption=Linear}
    boneToModify {boneName=dyn_clothC_r_01}
    constraintSetup [{targetBone {boneName=dyn_clothB_r_01}}, {targetBone {boneName=pelvis}}]
    constraintWeights [1,1]
   }
   n26 = "Constraint - calf_twist_armor_l (RR)"!(ComponentPose=@n25, Alpha=1, ConstraintWeights_0=0.4, ConstraintWeights_1:float=0.6) -> (Pose) {
    alphaBoolBlend {blendOption=Linear}
    boneToModify {boneName=calf_twist_armor_l}
    constraintSetup [{transformType=Rotation, targetBone {boneName=thigh_twist_01_l}}, {transformType=Rotation, targetBone {boneName=calf_l}}]
    constraintWeights [1,1]
   }
   n27 = "Constraint - calf_twist_armor_r (RR)"!(ComponentPose=@n26, Alpha=1, ConstraintWeights_0=0.4, ConstraintWeights_1:float=0.6) -> (Pose) {
    alphaBoolBlend {blendOption=Linear}
    boneToModify {boneName=calf_twist_armor_r}
    constraintSetup [{transformType=Rotation, targetBone {boneName=thigh_r}}, {transformType=Rotation, targetBone {boneName=calf_r}}]
    constraintWeights [1,1]
   }
   n28 = "Constraint - shldr_pad_l_base (PP)"(ComponentPose=@n27, Alpha=1, ConstraintWeights_0=0.6, ConstraintWeights_1=0.4) {
    alphaBoolBlend {blendOption=Linear}
    boneToModify {boneName=shldr_pad_l_base}
    constraintSetup [{targetBone {boneName=clavicle_l}}, {targetBone {boneName=upperarm_l}}]
    constraintWeights [1,1]
   }
   n29 = "Constraint - shldr_pad_r_base (PP)"(ComponentPose=@n28, Alpha=1, ConstraintWeights_0=0.6, ConstraintWeights_1=0.4) {
    alphaBoolBlend {blendOption=Linear}
    boneToModify {boneName=shldr_pad_r_base}
    constraintSetup [{targetBone {boneName=clavicle_r}}, {targetBone {boneName=upperarm_twist_01_r}}]
    constraintWeights [1,1]
   }
   n30 = "Transform (Modify) Bone - Bone: shldr_pad_l"(ComponentPose=@n29, Alpha=1, Translation="0,0,0", Rotation="5,5,-10", Scale="1,1,1") {
    rotationMode=BMM_Additive
    rotationSpace=BCS_BoneSpace
    scaleMode=BMM_Additive
    scaleSpace=BCS_BoneSpace
    translationMode=BMM_Additive
    translationSpace=BCS_BoneSpace
    alphaBoolBlend {blendOption=Linear}
    boneToModify {boneName=shldr_pad_l}
    rotation=<rot 5 5 -10>
   }
   n31 = "Transform (Modify) Bone - Bone: shldr_pad_r"(ComponentPose=@n30, Alpha=1, Translation="0,0,0", Rotation="5,6,-10", Scale="1,1,1") {
    rotationMode=BMM_Additive
    rotationSpace=BCS_BoneSpace
    scaleMode=BMM_Additive
    scaleSpace=BCS_BoneSpace
    translationMode=BMM_Additive
    translationSpace=BCS_BoneSpace
    alphaBoolBlend {blendOption=Linear}
    boneToModify {boneName=shldr_pad_r}
    rotation=<rot 5 6 -10>
   }
   n32 = "Transform (Modify) Bone - Bone: shldr_pad_l"(ComponentPose=@n31, Alpha=Get Shoulder_rot_back_l() -> (Shoulder_rot_back_l:double), Translation="0,0,0", Rotation="0,5,20", Scale="1,1,1") {
    rotationMode=BMM_Additive
    rotationSpace=BCS_BoneSpace
    scaleMode=BMM_Additive
    scaleSpace=BCS_BoneSpace
    translationMode=BMM_Additive
    translationSpace=BCS_BoneSpace
    alphaBoolBlend {blendOption=Linear}
    boneToModify {boneName=shldr_pad_l}
    rotation {roll=20, yaw=5}
   }
   n33 = "Transform (Modify) Bone - Bone: shldr_pad_r"(ComponentPose=@n32, Alpha=Get Shoulder_rot_back_r() -> (Shoulder_rot_back_r:double), Translation="0,0,0", Rotation="0,0,90", Scale="1,1,1") {
    rotationMode=BMM_Additive
    rotationSpace=BCS_BoneSpace
    scaleMode=BMM_Additive
    scaleSpace=BCS_BoneSpace
    translationMode=BMM_Additive
    translationSpace=BCS_BoneSpace
    alphaBoolBlend {blendOption=Linear}
    boneToModify {boneName=shldr_pad_r}
    rotation {roll=90}
   }
   n34 = "Transform (Modify) Bone - Bone: calf_twist_armor_l"(ComponentPose=@n33, Alpha=Get thigh_rot_l(), Translation="0,-3,0", Rotation="0,0,0", Scale="1,1,1") {
    rotationMode=BMM_Additive
    rotationSpace=BCS_BoneSpace
    scaleMode=BMM_Additive
    scaleSpace=BCS_BoneSpace
    translationMode=BMM_Additive
    translationSpace=BCS_BoneSpace
    alphaBoolBlend {blendOption=Linear}
    boneToModify {boneName=calf_twist_armor_l}
    translation {y=-3}
   }
   n35 = "Transform (Modify) Bone - Bone: calf_twist_armor_r"(ComponentPose=@n34, Alpha=Get thigh_rot_r(), Translation="0,3,0", Rotation="0,0,0", Scale="1,1,1") {
    rotationSpace=BCS_BoneSpace
    scaleMode=BMM_Additive
    scaleSpace=BCS_BoneSpace
    translationMode=BMM_Additive
    translationSpace=BCS_BoneSpace
    alphaBoolBlend {blendOption=Linear}
    boneToModify {boneName=calf_twist_armor_r}
    translation {y=3}
   }
   n36 = Get leg_rot_l()
   n37 = "Transform (Modify) Bone - Bone: dyn_cloth_l_01"(ComponentPose=@n35, Alpha=@n36, Translation="0,0,-2", Rotation="-30,-10,0", Scale="1,1,1") {
    rotationMode=BMM_Additive
    rotationSpace=BCS_BoneSpace
    scaleMode=BMM_Additive
    scaleSpace=BCS_BoneSpace
    translationMode=BMM_Additive
    translationSpace=BCS_BoneSpace
    alphaBoolBlend {blendOption=Linear}
    boneToModify {boneName=dyn_cloth_l_01}
    rotation {pitch=-30, yaw=-10}
    translation {z=-2}
   }
   n38 = "Transform (Modify) Bone - Bone: dyn_clothA_l_01"(ComponentPose=@n37, Alpha=@n36, Translation="-5,0,-5", Rotation="-25,0,-40", Scale="1,1,1") {
    rotationMode=BMM_Additive
    rotationSpace=BCS_BoneSpace
    scaleMode=BMM_Additive
    scaleSpace=BCS_BoneSpace
    translationMode=BMM_Additive
    translationSpace=BCS_BoneSpace
    alphaBoolBlend {blendOption=Linear}
    boneToModify {boneName=dyn_clothA_l_01}
    rotation {pitch=-25, roll=-40}
    translation {x=-5, z=-5}
   }
   n39 = "Transform (Modify) Bone - Bone: dyn_clothB_l_01"(ComponentPose=@n38, Alpha=@n36, Translation="0,0,0", Rotation="0,0,0", Scale="1,1,1") {
    rotationMode=BMM_Additive
    rotationSpace=BCS_BoneSpace
    scaleMode=BMM_Additive
    scaleSpace=BCS_BoneSpace
    translationMode=BMM_Additive
    translationSpace=BCS_BoneSpace
    alphaBoolBlend {blendOption=Linear}
    boneToModify {boneName=dyn_clothB_l_01}
   }
   n40 = "Transform (Modify) Bone - Bone: thigh_l_rot"(ComponentPose=@n39, Alpha=@n36, Translation="-5,10,-2", Rotation="-20,-5,0", Scale="1,1,1") {
    rotationMode=BMM_Additive
    rotationSpace=BCS_BoneSpace
    scaleMode=BMM_Additive
    scaleSpace=BCS_BoneSpace
    translationMode=BMM_Additive
    translationSpace=BCS_BoneSpace
    alphaBoolBlend {blendOption=Linear}
    boneToModify {boneName=thigh_l_rot}
    rotation {pitch=-20, yaw=-5}
    translation=<-5 10 -2>
   }
   n41 = Get leg_rot_r() -> (leg_rot_r:double)
   n42 = "Transform (Modify) Bone - Bone: thigh_r_rot"(ComponentPose=@n40, Alpha=@n41, Translation="0,0,2", Rotation="15,-60,0", Scale="1,1,1") {
    rotationMode=BMM_Additive
    rotationSpace=BCS_BoneSpace
    scaleMode=BMM_Additive
    scaleSpace=BCS_BoneSpace
    translationMode=BMM_Additive
    translationSpace=BCS_BoneSpace
    alphaBoolBlend {blendOption=Linear}
    boneToModify {boneName=thigh_r_rot}
    rotation {pitch=15, yaw=-60}
    translation {z=2}
   }
   n43 = "Transform (Modify) Bone - Bone: dyn_clothA_r_01"(ComponentPose=@n42, Alpha=@n41, Translation="0,0,0", Rotation="0,0,0", Scale="1,1,1") {
    rotationMode=BMM_Additive
    rotationSpace=BCS_BoneSpace
    scaleMode=BMM_Additive
    scaleSpace=BCS_BoneSpace
    translationMode=BMM_Additive
    translationSpace=BCS_BoneSpace
    alphaBoolBlend {blendOption=Linear}
    boneToModify {boneName=dyn_clothA_r_01}
   }
   n44 = "Transform (Modify) Bone - Bone: shldr_pad_l"(ComponentPose=@n43, Alpha=Get Shoulder_rot_f_l() -> (Shoulder_rot_f_l:double), Translation="0,0,0", Rotation="0,-10,-40", Scale="1,1,1") {
    rotationMode=BMM_Additive
    rotationSpace=BCS_BoneSpace
    scaleMode=BMM_Additive
    scaleSpace=BCS_BoneSpace
    translationMode=BMM_Additive
    translationSpace=BCS_BoneSpace
    alphaBoolBlend {blendOption=Linear}
    boneToModify {boneName=shldr_pad_l}
    rotation {roll=-40, yaw=-10}
   }
   n45 = "Transform (Modify) Bone - Bone: shldr_pad_r"(ComponentPose=@n44, Alpha=Get Shoulder_rot_f_r() -> (Shoulder_rot_f_r:double), Translation="1,2,1", Rotation="0,-10,-40", Scale="1,1,1") {
    rotationMode=BMM_Additive
    rotationSpace=BCS_BoneSpace
    scaleMode=BMM_Additive
    scaleSpace=BCS_BoneSpace
    translationMode=BMM_Additive
    translationSpace=BCS_BoneSpace
    alphaBoolBlend {blendOption=Linear}
    boneToModify {boneName=shldr_pad_r}
    rotation {roll=-40, yaw=-10}
    translation=<1 2 1>
   }
   n46 = Get Shoulder_rot_up_l() -> (Shoulder_rot_up_l:double)
   n47 = "Transform (Modify) Bone - Bone: shldr_pad_l"(ComponentPose=@n45, Alpha=@n46, Translation="1,-4,3", Rotation="0,-40,-40", Scale="1,1,1") {
    rotationMode=BMM_Additive
    rotationSpace=BCS_BoneSpace
    scaleMode=BMM_Additive
    scaleSpace=BCS_BoneSpace
    translationMode=BMM_Additive
    translationSpace=BCS_BoneSpace
    alphaBoolBlend {blendOption=Linear}
    boneToModify {boneName=shldr_pad_l}
    rotation {roll=-40, yaw=-40}
    translation=<1 -4 3>
   }
   n48 = "Transform (Modify) Bone - Bone: upperarmDn_l"(ComponentPose=@n47, Alpha=@n46, Translation="-2,0,4", Rotation="0,0,0", Scale="1,1,1") {
    rotationMode=BMM_Additive
    rotationSpace=BCS_BoneSpace
    scaleMode=BMM_Additive
    scaleSpace=BCS_BoneSpace
    translationMode=BMM_Additive
    translationSpace=BCS_BoneSpace
    alphaBoolBlend {blendOption=Linear}
    boneToModify {boneName=upperarmDn_l}
    translation {x=-2, z=4}
   }
   n49 = Get Shoulder_rot_up_r() -> (Shoulder_rot_up_r:double)
   n50 = "Transform (Modify) Bone - Bone: upperarmDn_r"(ComponentPose=@n48, Alpha=@n49, Translation="2,0,-4", Rotation="0,0,0", Scale="1,1,1") {
    rotationMode=BMM_Additive
    rotationSpace=BCS_BoneSpace
    scaleMode=BMM_Additive
    scaleSpace=BCS_BoneSpace
    translationMode=BMM_Additive
    translationSpace=BCS_BoneSpace
    alphaBoolBlend {blendOption=Linear}
    boneToModify {boneName=upperarmDn_r}
    translation {x=2, z=-4}
   }
   n51 = "Transform (Modify) Bone - Bone: shldr_pad_r"(ComponentPose=@n50, Alpha=@n49, Translation="-1,2.5,0", Rotation="0,-40,-40", Scale="1,1,1") {
    rotationMode=BMM_Additive
    rotationSpace=BCS_BoneSpace
    scaleMode=BMM_Additive
    scaleSpace=BCS_BoneSpace
    translationMode=BMM_Additive
    translationSpace=BCS_BoneSpace
    alphaBoolBlend {blendOption=Linear}
    boneToModify {boneName=shldr_pad_r}
    rotation {roll=-40, yaw=-40}
    translation=<-1 2.5>
   }
   n52 = "Transform (Modify) Bone - Bone: upperarmDn_r"(ComponentPose=@n51, Alpha=Get Shoulder_rot_dn_r() -> (Shoulder_rot_dn_r:double), Translation="2,0,0", Rotation="0,0,0", Scale="1,1,1") {
    rotationMode=BMM_Additive
    rotationSpace=BCS_BoneSpace
    scaleMode=BMM_Additive
    scaleSpace=BCS_BoneSpace
    translationMode=BMM_Additive
    translationSpace=BCS_BoneSpace
    alphaBoolBlend {blendOption=Linear}
    boneToModify {boneName=upperarmDn_r}
    translation {x=2}
   }
   n53 = "Transform (Modify) Bone - Bone: upperarmDn_l"(ComponentPose=@n52, Alpha=Get Shoulder_rot_dn_l() -> (Shoulder_rot_dn_l:double), Translation="-2,0,0", Rotation="0,0,0", Scale="1,1,1") {
    rotationMode=BMM_Additive
    rotationSpace=BCS_BoneSpace
    scaleMode=BMM_Additive
    scaleSpace=BCS_BoneSpace
    translationMode=BMM_Additive
    translationSpace=BCS_BoneSpace
    alphaBoolBlend {blendOption=Linear}
    boneToModify {boneName=upperarmDn_l}
    translation {x=-2}
   }
   n54 = "Transform (Modify) Bone - Bone: dyn_clothE_l_00"(ComponentPose=@n53, Alpha=Get leg_rot_back_l() -> (leg_rot_back_l:double), Translation="0,0,0", Rotation="0,0,-60", Scale="1,1,1") {
    rotationMode=BMM_Additive
    rotationSpace=BCS_BoneSpace
    scaleMode=BMM_Additive
    scaleSpace=BCS_BoneSpace
    translationMode=BMM_Additive
    translationSpace=BCS_BoneSpace
    alphaBoolBlend {blendOption=Linear}
    boneToModify {boneName=dyn_clothE_l_00}
    rotation {roll=-60}
   }
   n55 = "Transform (Modify) Bone - Bone: dyn_clothE_r_00"(ComponentPose=@n54, Alpha=Get leg_rot_back_r() -> (leg_rot_back_r:double), Translation="0,0,0", Rotation="0,0,-60", Scale="1,1,1") {
    rotationMode=BMM_Additive
    rotationSpace=BCS_BoneSpace
    scaleMode=BMM_Additive
    scaleSpace=BCS_BoneSpace
    translationMode=BMM_Additive
    translationSpace=BCS_BoneSpace
    alphaBoolBlend {blendOption=Linear}
    boneToModify {boneName=dyn_clothE_r_00}
    rotation {roll=-60}
   }
   Save cached pose 'AnimCache'(Pose=Component To Local(ComponentPose=@n55)) {cacheName=AnimCache}
  ]}
 EventGraph {nodes [
   Event BlueprintUpdateAnimation()
   n1 = Get Skel_Mesh()
   > Set thigh_rot_r(thigh_rot_r=GetAlphaFromReferenceTransformDelta(SkeletalMeshComponent=@n1, BoneName=calf_r, BaseName=thigh_r, InputRange=(X=0,Y=110), TwistAxis="0,0,1"))
   > Set thigh_rot_l(thigh_rot_l=GetAlphaFromReferenceTransformDelta(SkeletalMeshComponent=@n1, BoneName=calf_l, BaseName=thigh_l, InputRange=(X=0,Y=110), TwistAxis="0,0,1"))
   n2 = Get Upperarm_Mesh() -> (Upperarm_Mesh:Skeletal Mesh Component Object Reference)
   > Set Shoulder_rot_f_r(Shoulder_rot_f_r:double=GetAlphaFromReferenceTransformDelta(SkeletalMeshComponent=@n2, BoneName=upperarm_r, BaseName=clavicle_r, InputRange=(X=0,Y=90), TwistAxis="0,0,1")) -> (Output_Get:double)
   > Set Shoulder_rot_f_l(Shoulder_rot_f_l:double=GetAlphaFromReferenceTransformDelta(SkeletalMeshComponent=@n2, BoneName=upperarm_l, BaseName=clavicle_l, InputRange=(X=0,Y=90), TwistAxis="0,0,1")) -> (Output_Get:double)
   > Set Shoulder_rot_back_l(Shoulder_rot_back_l:double=GetAlphaFromReferenceTransformDelta(SkeletalMeshComponent=@n2, BoneName=upperarm_r, BaseName=clavicle_r, InputRange=(X=0,Y=-90), TwistAxis="0,0,1")) -> (Output_Get:double)
   > Set Shoulder_rot_back_r(Shoulder_rot_back_r:double=GetAlphaFromReferenceTransformDelta(SkeletalMeshComponent=@n2, BoneName=upperarm_l, BaseName=clavicle_l, InputRange=(X=0,Y=-90), TwistAxis="0,0,1")) -> (Output_Get:double)
   n3 = Get Upperarm_Mesh()
   > Set Shoulder_rot_up_r(Shoulder_rot_up_r:double=GetAlphaFromReferenceTransformDelta(SkeletalMeshComponent=@n3, BoneName=upperarm_r, BaseName=clavicle_r, DeltaComponentType=EulerY, InputRange=(X=0,Y=90), TwistAxis="0,0,1")) -> (Output_Get:double)
   > Set Shoulder_rot_up_l(Shoulder_rot_up_l:double=GetAlphaFromReferenceTransformDelta(SkeletalMeshComponent=@n3, BoneName=upperarm_l, BaseName=clavicle_l, DeltaComponentType=EulerY, InputRange=(X=0,Y=90), TwistAxis="0,0,1")) -> (Output_Get:double)
   > Set Shoulder_rot_dn_r(Shoulder_rot_dn_r:double=GetAlphaFromReferenceTransformDelta(SkeletalMeshComponent=@n3, BoneName=upperarm_r, BaseName=clavicle_r, InputRange=(X=-15,Y=-90), TwistAxis="0,0,1")) -> (Output_Get:double)
   > Set Shoulder_rot_dn_l(Shoulder_rot_dn_l:double=GetAlphaFromReferenceTransformDelta(SkeletalMeshComponent=@n3, BoneName=upperarm_l, BaseName=clavicle_l, InputRange=(X=-15,Y=-90), TwistAxis="0,0,1")) -> (Output_Get:double)
   n4 = Get Skel_Mesh()
   > Set leg_rot_r(leg_rot_r:double=GetAlphaFromReferenceTransformDelta(SkeletalMeshComponent=@n4, BoneName=thigh_r, BaseName=pelvis, InputRange=(X=0,Y=-150), TwistAxis="0,0,1")) -> (Output_Get:double)
   > Set leg_rot_l(leg_rot_l:double=GetAlphaFromReferenceTransformDelta(SkeletalMeshComponent=@n4, BoneName=thigh_twist_01_l, BaseName=pelvis, InputRange=(X=0,Y=-150), TwistAxis="0,0,1")) -> (Output_Get:double)
   > Set leg_rot_l_bag(leg_rot_l_bag:double=GetAlphaFromReferenceTransformDelta(SkeletalMeshComponent=@n4, BoneName=thigh_l, BaseName=pelvis, InputRange=(X=0,Y=-90), TwistAxis="0,0,1")) -> (Output_Get:double)
   n5 = Get Skel_Mesh()
   > Set leg_rot_back_l(leg_rot_back_l:double=GetAlphaFromReferenceTransformDelta(SkeletalMeshComponent=@n5, BoneName=thigh_l, BaseName=pelvis, InputRange=(X=0,Y=150), TwistAxis="0,0,1")) -> (Output_Get:double)
   > Set leg_rot_back_r(leg_rot_back_r:double=GetAlphaFromReferenceTransformDelta(SkeletalMeshComponent=@n5, BoneName=thigh_r, BaseName=pelvis, InputRange=(X=0,Y=150), TwistAxis="0,0,1")) -> (Output_Get:double)
   Event BlueprintInitializeAnimation()
   > Set Skel_Mesh(Skel_Mesh=GetOwnerSkeletalMeshComponent())
   > n6 = Cast To PlayerPawn_Athena(Object=TryGetPawnOwner())
   > n7 = Set Pawn Owner(Pawn Owner:Fort Player Pawn Athena Object Reference=@n6) -> (Output_Get:Fort Player Pawn Athena Object Reference)
   > Set Upperarm_Mesh(Upperarm_Mesh:Skeletal Mesh Component Object Reference=GetSkeletalMeshForPartType(self=@n7)) -> (Output_Get:Skeletal Mesh Component Object Reference)
   Get Skel_Mesh()
   Get Skel_Mesh()
  ]}
]
variables [
 leg_rot_back_l {type=number}
 leg_rot_back_r {type=number}
 leg_rot_l {type=number}
 leg_rot_l_bag {type=number}
 leg_rot_r {type=number}
 Pawn Owner {type=FortPlayerPawnAthena}
 Shoulder_rot_back_l {type=number}
 Shoulder_rot_back_r {type=number}
 Shoulder_rot_dn_l {type=number}
 Shoulder_rot_dn_r {type=number}
 Shoulder_rot_f_l {type=number}
 Shoulder_rot_f_r {type=number}
 Shoulder_rot_up_l {type=number}
 Shoulder_rot_up_r {type=number}
 Skel_Mesh {type=SkeletalMeshComponent}
 thigh_rot_l {type=number}
 thigh_rot_r {type=number}
 Upperarm_Mesh {type=SkeletalMeshComponent}
]
