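# Physics asset for the head-area parts of the female medium Taco Knight cosmetic body: helmet, chili and chili leaves.
# Capsule bodies cover the head, helmet, chili segments and four leaves, with two capsules on the helmet.
# Constraints chain the chili from the head through its base to the tip and branch the leaves off that base.
# Limit angles are near zero along the chili and wider on the leaves, while the helmet-to-head joint locks twist and one swing axis.
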PhysicsAsset /BRCosmetics/Characters/Player/Female/Medium/Bodies/F_MED_Taco_Knight/Meshes/Parts/F_MED_Taco_Knight_Physics
bodies {
 chili_01 [{shapeType=Capsule, length=10.1978, radius=4.4381, center=<4.7825 0.8985 -0.0759>, rotation=<rot 0 78.4053 -88.6766>}]
 chili_01_base [{shapeType=Capsule, length=9.6029, radius=4.7856, center=<4.2333 0.573 0.2023>, rotation=<rot 5.5904 -86.2072 -81.7908>}]
 chili_02 [{shapeType=Capsule, length=5.4611, radius=3.5195, center=<1.7651 0.1676 -0.2101>, rotation=<rot -0.2091 -91.5609 -89.126>}]
 chili_03 [{shapeType=Capsule, length=2.9981, radius=2.7438, center=<0.7229 1.6412 -0.7134>, rotation=<rot 0 -16.5091 -81.1208>}]
 chili_leaf_01 [{shapeType=Capsule, length=3.7982, radius=1.7903, center=<-0.7223 1.2047 -0.1567>, rotation=<rot 3.9008 -173.9581 -87.0421>}]
 chili_leaf_02 [{shapeType=Capsule, length=2.6657, radius=2.3419, center=<-0.6292 1.0191 0.0627>, rotation=<rot 3.4566 -165.244 -81.6528>}]
 chili_leaf_03 [{shapeType=Capsule, length=2.3009, radius=1.6608, center=<-0.8121 3.0898 0.0537>, rotation=<rot 0 42.9149 -83.9879>}]
 chili_leaf_04 [{shapeType=Capsule, length=2.3271, radius=1.6173, center=<-0.7216 3.2843 -0.2878>, rotation=<rot 0 32.7987 -80.3517>}]
 dny_head [{shapeType=Capsule, length=18.9692, radius=7.6887, center=<13.4912 -3.4036 -0.2462>, rotation=<rot 88.9496 0.0001 10.087>}]
 helmet [
  {shapeType=Capsule, length=22.2359, radius=4.4055, center=<8.3009 7.3197 -0.0178>, rotation=<rot -0.4289 -90 -0.0439>}
  {shapeType=Capsule, length=4.5647, radius=3.3995, rotation=<rot 89.2745 -179.9999 -179.9999>}
 ]
}
constraints [
 {
  bone1Name=helmet
  bone2Name=dny_head
  swing1LimitDegrees=5
  swing2LimitDegrees=45
  swing2Motion=Locked
  twistLimitDegrees=45
  twistMotion=Locked
 }
 {bone1Name=chili_01, bone2Name=chili_01_base, swing1LimitDegrees=0.1, swing2LimitDegrees=0.1, twistLimitDegrees=0.1}
 {bone1Name=chili_02, bone2Name=chili_01, swing1LimitDegrees=0.1, swing2LimitDegrees=0.1, twistLimitDegrees=0.1}
 {bone1Name=chili_03, bone2Name=chili_02, swing1LimitDegrees=2, swing2LimitDegrees=2, twistLimitDegrees=2}
 {bone1Name=chili_leaf_01, bone2Name=chili_01_base, swing1LimitDegrees=10, swing2LimitDegrees=30, twistLimitDegrees=10}
 {bone1Name=chili_leaf_02, bone2Name=chili_01_base, swing1LimitDegrees=10, swing2LimitDegrees=30, twistLimitDegrees=10}
 {bone1Name=chili_leaf_03, bone2Name=chili_01_base, swing1LimitDegrees=10, swing2LimitDegrees=30, twistLimitDegrees=10}
 {bone1Name=chili_leaf_04, bone2Name=chili_01_base, swing1LimitDegrees=10, swing2LimitDegrees=30, twistLimitDegrees=10}
 {bone1Name=chili_01_base, bone2Name=dny_head, swing1LimitDegrees=0.1, swing2LimitDegrees=0.1, twistLimitDegrees=0.1}
]
