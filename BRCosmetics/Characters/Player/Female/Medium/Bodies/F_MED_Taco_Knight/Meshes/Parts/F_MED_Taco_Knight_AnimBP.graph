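# Post-process style AnimBP for the F_MED_Taco_Knight cosmetic body part, copying the attached parent's pose and adding secondary head and helmet motion.
# Head and right-shoulder rotation alphas, computed each update, drive dny_head constraints and additive offsets, though both shoulder offsets are zero.
# A RigidBody simulation relative to spine_05 follows and is cached.
# Helmet AnimDynamics blends over that cached pose while bIsAccelerating2D is true.
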AnimBlueprint /BRCosmetics/Characters/Player/Female/Medium/Bodies/F_MED_Taco_Knight/Meshes/Parts/F_MED_Taco_Knight_AnimBP
target_skeleton=&/F_MED_Taco_Knight_Skeleton
graphs [
 AnimGraph {nodes [
   n1 = Anim Dynamics - Bone: helmet(ComponentPose=Local To Component(LocalPose=Use cached pose 'AnimCache'()), Alpha=1, ExternalForce="0,0,0") {
    angularDampingOverride=0.9
    bGravityOverrideInSimSpace=true
    bOverrideAngularDamping=true
    bUseGravityOverride=true
    alphaBoolBlend {blendOption=Linear}
    boundBone {boneName=helmet}
    componentAppliedLinearAccClamp=<1000 1000 1000>
    componentLinearAccScale=<1 1 1>
    componentLinearVelScale=<1 1 1>
    gravityOverride {z=-980}
    physicsBodyDefinitions [{
      sphereCollisionRadius=0
      boundBone {boneName=helmet}
      boxExtents=<8 15 12>
      constraintSetup {angularLimitsMax=<0 0 4>, angularLimitsMin=<0 0 -11>, angularTarget=<0 0 0>}
      localJointOffset=<1 0 0>
     }]
   }
   Output Pose(Result=Two Way Blend!(A=Use cached pose 'AnimCache'(), B=Component To Local(ComponentPose=@n1), bAlphaBoolEnabled:bool=Get bIsAccelerating2D()) -> (Pose) {blendNode {alphaInputType=Bool, alphaBoolBlend {blendInTime=0.05, blendOption=Cubic, blendOutTime=0.1}}}) {layerGroup=None}
   n2 = "Constraint - dny_head (RR)"!(
    ComponentPose=Local To Component(LocalPose=Copy Pose From Mesh() {bUseAttachedParent=true})
    Alpha=1
    ConstraintWeights_0=0.4
    ConstraintWeights_1:float=0.6
   ) -> (Pose) {
    alphaBoolBlend {blendOption=Linear}
    boneToModify {boneName=dny_head}
    constraintSetup [{transformType=Rotation, targetBone {boneName=neck_02}}, {transformType=Rotation, targetBone {boneName=head}}]
    constraintWeights [1,1]
   }
   n3 = Get head_rot_up() -> (head_rot_up:double)
   n4 = "Transform (Modify) Bone - Bone: dny_head"(ComponentPose=@n2, Alpha=@n3, Translation="0,0,0", Rotation="0,10,0", Scale="1,1,1") {
    rotationMode=BMM_Additive
    rotationSpace=BCS_BoneSpace
    scaleMode=BMM_Additive
    scaleSpace=BCS_BoneSpace
    translationMode=BMM_Additive
    translationSpace=BCS_BoneSpace
    alphaBoolBlend {blendOption=Linear}
    boneToModify {boneName=dny_head}
    rotation {yaw=10}
   }
   n5 = "Constraint - dny_head (RR)"(ComponentPose=@n4, Alpha=@n3, ConstraintWeights_0=0.3, ConstraintWeights_1=0.7) {
    alphaBoolBlend {blendOption=Linear}
    boneToModify {boneName=dny_head}
    constraintSetup [{transformType=Rotation, targetBone {boneName=neck_02}}, {transformType=Rotation, targetBone {boneName=spine_05}}]
    constraintWeights [1,1]
   }
   n6 = "Transform (Modify) Bone - Bone: dny_head"(ComponentPose=@n5, Alpha=Get Shoulder_rot_f_r() -> (Shoulder_rot_f_r:double), Translation="0,0,0", Rotation="0,0,0", Scale="1,1,1") {
    rotationMode=BMM_Additive
    rotationSpace=BCS_BoneSpace
    scaleMode=BMM_Additive
    scaleSpace=BCS_BoneSpace
    translationMode=BMM_Additive
    translationSpace=BCS_BoneSpace
    alphaBoolBlend {blendOption=Linear}
    boneToModify {boneName=dny_head}
   }
   n7 = "Transform (Modify) Bone - Bone: dny_head"(ComponentPose=@n6, Alpha=Get Shoulder_rot_up_r() -> (Shoulder_rot_up_r:double), Translation="0,0,0", Rotation="0,0,0", Scale="1,1,1") {
    rotationMode=BMM_Additive
    rotationSpace=BCS_BoneSpace
    scaleMode=BMM_Additive
    scaleSpace=BCS_BoneSpace
    translationMode=BMM_Additive
    translationSpace=BCS_BoneSpace
    alphaBoolBlend {blendOption=Linear}
    boneToModify {boneName=dny_head}
   }
   n8 = RigidBody(ComponentPose=@n7) {
    bClampLinearTranslationLimitToRefPose=true
    bOverrideWorldGravity=true
    overridePhysicsAsset=&/F_MED_Taco_Knight_Physics
    simulationSpace=BaseBoneSpace
    alphaBoolBlend {blendOption=Linear}
    baseBoneRef {boneName=spine_05}
    componentAppliedLinearAccClamp=<800 800 800>
    componentLinearAccScale=<0.35 0.35 0.35>
    componentLinearVelScale=<0.3 0.3 0.3>
    overrideWorldGravity {z=-980}
   }
   Save cached pose 'AnimCache'(Pose=Component To Local(ComponentPose=@n8)) {cacheName=AnimCache}
  ]}
 EventGraph {nodes [
   Event BlueprintUpdateAnimation()
   n1 = Get Skel_Mesh()
   > Set head_rot_up(head_rot_up:double=GetAlphaFromReferenceTransformDelta(SkeletalMeshComponent=@n1, BoneName=head, BaseName=neck_02, InputRange=(X=0,Y=-50), TwistAxis="0,0,1")) -> (Output_Get:double)
   > Set Shoulder_rot_f_r(Shoulder_rot_f_r:double=GetAlphaFromReferenceTransformDelta(SkeletalMeshComponent=@n1, BoneName=upperarm_r, BaseName=clavicle_r, InputRange=(X=0,Y=90), TwistAxis="0,0,1")) -> (Output_Get:double)
   > Set Shoulder_rot_up_r(Shoulder_rot_up_r:double=GetAlphaFromReferenceTransformDelta(SkeletalMeshComponent=@n1, BoneName=upperarm_r, BaseName=clavicle_r, DeltaComponentType=EulerY, InputRange=(X=0,Y=60), TwistAxis="0,0,1")) -> (Output_Get:double)
   Event BlueprintInitializeAnimation()
   > Set Skel_Mesh(Skel_Mesh=GetOwnerSkeletalMeshComponent())
  ]}
]
variables [
 head_rot_up {type=number, default=1}
 Shoulder_rot_f_r {type=number}
 Shoulder_rot_up_r {type=number}
 Skel_Mesh {type=SkeletalMeshComponent}
]
